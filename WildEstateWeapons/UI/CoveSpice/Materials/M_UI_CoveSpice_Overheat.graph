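# Translucent UI material for the CoveSpice weapon's overheat meter, drawn from channel-packed distance fields in one texture.
# AnimateProgress drives a fill edge between ProgressStart and ProgressEnd on rotating UVs and shifts the fill from white toward FilledColor.
# Red-channel thresholds form backplate, fill and glow, with distortion and glow rising near full heat.
# The green channel yields a distorted laser beam, shown once progress reaches OverheatThreshold.
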Material /WildEstateWeapons/UI/CoveSpice/Materials/M_UI_CoveSpice_Overheat
graph {
 properties {bAutomaticallySetUsageInEditor=false, blendMode=BLEND_Translucent, materialDomain=MD_UI, usedShadingModels=Default Lit}
 nodes [
  Constant() {r=0.5}
  n1 = VectorParameter() {parameterName=Backplate Color, defaultValue=<rgba 0 0 0 1>}
  n2 = NamedRerouteUsage()
  n3 = VectorParameter() {parameterName=FilledColor, defaultValue=<rgba 1 0 0.0833 1>}
  n4 = ScalarParameter() {
   desc="Property for animating progress value change (from ProgressStart to ProgressEnd)\r\n0 - 1"
   group=Progress
   parameterName=AnimateProgress
   sliderMax=1
   sortPriority=3
  }
  n5 = NamedRerouteUsage()
  n6 = NamedRerouteUsage()
  n7 = LinearInterpolate(
   A=LinearInterpolate(A=Multiply(A=@n1.RGB, B=@n2), B=LinearInterpolate(B=@n3.RGB, Alpha=@n4) {constA=1}, Alpha=Multiply(A=@n5, B=@n6))
   B=VectorParameter() {parameterName=BeamColor, defaultValue=<rgba 1 0.3368 0.2378 1>}.RGB
   Alpha=NamedRerouteUsage()
  ) {constAlpha=0}
  LinearInterpolate(A=@n7, B=VectorParameter() {parameterName=BeamColorStatic, defaultValue=<rgba 1 0.1024 0.1574 1>}.RGB, Alpha=NamedRerouteUsage())
  n8 = ScalarParameter() {
   desc="Property for animating progress value change (from ProgressStart to ProgressEnd)\r\n0 - 1"
   group=Progress
   parameterName=AnimateProgress
   sliderMax=1
   sortPriority=3
  }
  n9 = MaterialFunctionCall!(A=@n8, B=ScalarParameter() {defaultValue=0.95, parameterName=OverheatThreshold}) {materialFunction=/FortUILibrary/MaterialFunctions/Basics/MF_UI_BooleanOperations}
  Multiply(
   A=MaterialFunctionCall(
    DistanceField=TextureSample() {texture=/WildEstateWeapons/UI/CoveSpice/Textures/T_UI_CoveSpice_Overheat_128}.G
    Threshold=Constant() {r=0.5}
    Smoothing=Constant() {r=0.2}
   ) {materialFunction=&MF_UI_AntiAliasing}.Outer
   B=@n9.IsGreaterOrEqual
  )
  n10 = TextureCoordinate()
  n11 = Constant() {r=-2}
  n12 = Constant() {r=2}
  n13 = ScalarParameter() {
   desc="Property for animating progress value change (from ProgressStart to ProgressEnd)\r\n0 - 1"
   group=Progress
   parameterName=AnimateProgress
   sliderMax=1
   sortPriority=3
  }
  n14 = Multiply(A=Constant() {r=0.05}, B=SmoothStep(Value=@n13) {constMin=0.8, constValue=0.05}) {constB=0}
  n15 = TextureSample(UVs=Add(
    A=Subtract(A=@n10) {constB=0}
    B=MaterialFunctionCall!(UV, SpeedX=@n11, SpeedY=@n12, Intensity=@n14) {materialFunction=/WildEstateWeapons/UI/MaterialFunctions/MF_UI_WildEstate_Distort}.Result
   )) {texture=&T_UI_CoveSpice_Overheat_128}
  NamedRerouteDeclaration(None=MaterialFunctionCall(DistanceField=@n15.R, Threshold=Constant() {r=0.35}) {materialFunction=&MF_UI_AntiAliasing}.Outer) {name=Backplate, nodeColor=<rgba 0 0 0 1>}
  n16 = Multiply(A=NamedRerouteUsage(), B=NamedRerouteUsage())
  n17 = NamedRerouteUsage()
  n18 = Multiply(A=@n17) {constB=0.5}
  n19 = NamedRerouteUsage()
  NamedRerouteDeclaration(None=Add(A=LinearInterpolate(A=@n16, B=@n17, Alpha=@n18), B=Multiply(A=@n19))) {name=Opacity, nodeColor=<rgba 0 0 0 1>}
  n20 = ScalarParameter() {
   desc="Property for animating progress value change (from ProgressStart to ProgressEnd)\r\n0 - 1"
   group=Progress
   parameterName=AnimateProgress
   sliderMax=1
   sortPriority=3
  }
  n21 = MaterialFunctionCall(UVs=TextureCoordinate(), "Rotation Angle (0-1)"=LinearInterpolate(Alpha=@n20) {constA=0.31, constB=0.5}) {materialFunction=&CustomRotator}
  n22 = ScalarParameter() {
   desc="When progress value changes, it defines where change starts.\r\n0 - 1"
   group=Progress
   parameterName=ProgressStart
   sliderMax=1
   sortPriority=1
  }
  n23 = ScalarParameter() {
   defaultValue=0.92
   desc="When progress value changes, it defines where change ends.\r\n0 - 1"
   group=Progress
   parameterName=ProgressEnd
   sliderMax=1
   sortPriority=2
  }
  n24 = ScalarParameter() {
   desc="Property for animating progress value change (from ProgressStart to ProgressEnd)\r\n0 - 1"
   group=Progress
   parameterName=AnimateProgress
   sliderMax=1
   sortPriority=3
  }
  n25 = LinearInterpolate(A=@n22, B=@n23, Alpha=@n24)
  n26 = ComponentMask(None=Add(A=Multiply(A=@n21.Rotated Values) {constB=-1}, B=NamedRerouteDeclaration(None=@n25) {name=Progress, nodeColor=<rgba 0 0 0 1>}) {constB=0.5}) {g=true}
  NamedRerouteDeclaration(None=SmoothStep(Value=@n26) {constMax=0.02, constMin=-0.04}) {name=Glow Progress, nodeColor=<rgba 0 0 0 1>}
  n27 = VectorParameter() {parameterName=GlowColor, defaultValue=<rgba 0.5712 0 0.0476 1>}
  n28 = NamedRerouteUsage()
  NamedRerouteDeclaration(None=Max(A=@n7, B=Multiply(A=@n27.RGB, B=@n28))) {name=FinalColor, nodeColor=<rgba 0 0 0 1>}
  NamedRerouteDeclaration() {name=LaserBeamStatic, nodeColor=<rgba 0 0 0 1>}
  NamedRerouteDeclaration(None=MaterialFunctionCall(DistanceField=@n15.R, Threshold=Constant() {r=0.7}) {materialFunction=&MF_UI_AntiAliasing}.Outer) {name=FilledPart, nodeColor=<rgba 0 0 0 1>}
  n29 = ScalarParameter() {defaultValue=0.01, group=Progress_Layout, parameterName=ProgressEdgeSharpness, sliderMax=1, sortPriority=2}
  NamedRerouteDeclaration(None=Saturate(None=SmoothStep(Max=@n29, Value=@n26) {constMax=0.01})) {name=ProgressGradient, nodeColor=<rgba 0 0 0 1>}
  n30 = TextureCoordinate()
  n31 = Constant() {r=-3}
  n32 = Constant() {r=3}
  n33 = Constant() {r=0.28}
  n34 = TextureSample(UVs=Add(
    A=Subtract(A=@n30) {constB=0.05}
    B=MaterialFunctionCall(SpeedX=@n31, SpeedY=@n32, Intensity=@n33) {materialFunction=&MF_UI_WildEstate_Distort}.Result
   )) {texture=&T_UI_CoveSpice_Overheat_128}
  n35 = Constant() {r=0.55}
  NamedRerouteDeclaration(None=Multiply(
    A=MaterialFunctionCall(DistanceField=@n34.G, Threshold=@n35) {materialFunction=&MF_UI_AntiAliasing}.Outer
    B=@n9.IsGreaterOrEqual
   )) {name=LaserBeamAnimated, nodeColor=<rgba 0 0 0 1>}
  n36 = MaterialFunctionCall!(
   RGBA Channels=@n15.R
   Glow Min=ScalarParameter() {parameterName=GlowSizeMin}
   Glow Max=ScalarParameter() {defaultValue=0.5, parameterName=GlowSizeMax}
   Outline Thickness
  ) {materialFunction=/FortUILibrary/MaterialFunctions/Basics/MF_UI_ChannelPackedDistanceField}
  n37 = ScalarParameter() {
   desc="Property for animating progress value change (from ProgressStart to ProgressEnd)\r\n0 - 1"
   group=Progress
   parameterName=AnimateProgress
   sliderMax=1
   sortPriority=3
  }
  NamedRerouteDeclaration(None=Multiply(A=Power(Base=@n36.Fill) {constExponent=0.4}, B=SmoothStep(Value=@n37) {constMin=0.95})) {name=Glow, nodeColor=<rgba 1 0 0.0703 1>}
  NamedRerouteUsage()
  NamedRerouteUsage()
 ]
}
