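# Translucent UI material for the CoveSpice weapon reticle, drawing a procedural circular ring.
# A circle SDF sized by a scale parameter feeds a shape outline function with adjustable thickness.
# Two anti-aliased masks of that outline, one thresholded by ShadowThickness, are blended by ShadowOpacity and multiplied by ReticleOpacity.
# At the default ShadowOpacity, the blend resolves entirely to the shadow-threshold mask.
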Material /WildEstateWeapons/UI/CoveSpice/Materials/M_UI_CoveSpice_Reticle_Circle
graph {
 properties {bAutomaticallySetUsageInEditor=false, blendMode=BLEND_Translucent, materialDomain=MD_UI, usedShadingModels=Default Lit}
 nodes [
  n1 = MaterialFunctionCall!(
   DistanceField=MaterialFunctionCall(Size=ScalarParameter() {defaultValue=0.8, parameterName=Reticle Scale}) {materialFunction=&MF_UI_Circle}.SDF
   Threshold
   Thickness=ScalarParameter() {defaultValue=0.04, parameterName=Outline Tchickness}
   Alignment
   Smoothing
  ) {materialFunction=/FortUILibrary/MaterialFunctions/Shapes/Modifiers/MF_UI_ShapeOutline}
  n2 = ScalarParameter() {defaultValue=0.06, parameterName=ShadowThickness}
  Multiply(
   A=LinearInterpolate(
    A=MaterialFunctionCall(DistanceField=@n1.SDFOutline) {materialFunction=&MF_UI_AntiAliasing}.Inner
    B=MaterialFunctionCall(DistanceField=@n1.SDFOutline, Threshold=@n2) {materialFunction=&MF_UI_AntiAliasing}.Inner
    Alpha=ScalarParameter() {defaultValue=1, parameterName=ShadowOpacity}
   )
   B=ScalarParameter() {defaultValue=0.25, parameterName=ReticleOpacity}
  )
 ]
}
