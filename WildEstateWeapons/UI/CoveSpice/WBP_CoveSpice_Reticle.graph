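# Reticle extension widget for the CoveSpice weapon, showing target-in-range state and an overheat meter.
# Construct binds to the HUD passive ability and weapon overheat and firing events; a curve-table-gated looping timer can retry a failed ability bind.
# Range changes animate a reticle material parameter, while overheat ratio drives the meter material, which fades out at zero.
# Becoming visible refreshes the meter, including a leftover debug PrintString.
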WidgetBlueprint /WildEstateWeapons/UI/CoveSpice/WBP_CoveSpice_Reticle
animations [
 Anim_InRange {
  binding_count=1
  length_seconds=0.15
  object=&self:Anim_InRange
  sequence {bindings [Image_Reticle {
     id=#1
     possessed_class=Image
     tracks [{
       type=MovieSceneWidgetMaterialTrack
       sections [{type=MovieSceneParameterSection, end_frame=3, start_frame=0, channels [InRange {type=MovieSceneScriptingFloatChannel, keys [[0,0],[3,1]]}]}]
      }]
    }]}
 }
 Anim_HideCooldown {
  binding_count=1
  length_seconds=0.25
  object=&self:Anim_HideCooldown
  sequence {bindings [Image_Cooldown {
     id=#2
     possessed_class=Image
     tracks [{
       type=MovieSceneFloatTrack
       display_name=Render Opacity
       sections [{type=MovieSceneFloatSection, end_frame=5, start_frame=0, channels [None {type=MovieSceneScriptingFloatChannel, keys [[0,1],[5,0]]}]}]
      }]
    }]}
 }
]
graphs [
 BindHUDPassiveGA {nodes [
   BindHUDPassiveGA() {is=title K2Node_FunctionEntry}
   > n1 = Branch(Condition=IsValid(Object=Get HUDPassive() -> (HUDPassive:GA Cove Spice HUD Passive Base Object Reference)))
   Return Node!(execute=[@n7.then, @n1.then], bSuccess:bool=true)
   n2 = Cast To FortPlayerPawnAthena(execute=@n1.else, Object=GetOwningPlayerPawn!(self:User Widget Object Reference) -> (ReturnValue:Pawn Object Reference))
   > n3 = GetPrimaryAbilityInstanceOfAbilitySystemComponent!(
    execute
    AbilitySystemComponent:Fort Ability System Component Object Reference=GetFortAbilitySystemComponent!(Actor:Actor Object Reference=@n2) -> (ReturnValue:Fort Ability System Component Object Reference)
    AbilityClass:Fort Gameplay Ability Class Reference=/WildEstateWeapons/Gameplay/CoveSpice/Effects/GA_CoveSpice_HUD_Passive_Base.GA_CoveSpice_HUD_Passive_Base_C
   ) -> (then, ReturnValue:Gameplay Ability Object Reference)
   > n4 = Is Valid(InputObject=@n3)
   > n5 = Cast To GA_CoveSpice_HUD_Passive_Base!(execute, Object=@n3) -> (then, CastFailed, AsGA Cove Spice HUD Passive Base:GA Cove Spice HUD Passive Base Object Reference) {is=title K2Node_DynamicCast}
   > n6 = Set HUDPassive(HUDPassive:GA Cove Spice HUD Passive Base Object Reference=@n5) -> (Output_Get:GA Cove Spice HUD Passive Base Object Reference)
   > n7 = Bind Event to On Update Reticle(self:GA Cove Spice HUD Passive Base Object Reference=@n6, Delegate=Create Event()) {is=title K2Node_AddDelegate}
   Return Node(execute=[@n5.CastFailed, @n4.Is Not Valid, @n2.CastFailed], bSuccess=false)
  ]}
 EventGraph {nodes [
   StartFiring() {is=title K2Node_CustomEvent}
   > SetVisibility(self=Get Image_Cooldown() -> (Image_Cooldown:Image Object Reference), InVisibility=HitTestInvisible)
   > PlayAnimationReverse(InAnimation=Get Anim_HideCooldown() -> (Anim_HideCooldown:Widget Animation Object Reference))
   Event UpdateBypassHUDPassive() {is=title K2Node_CustomEvent}
   n1 = IsValid(Object=Get HUDPassive())
   n2 = Get ShouldBypassHUDPassiveFailure() -> (ShouldBypassHUDPassiveFailure:Scalable Float Structure)
   > n3 = Branch(Condition=AND Boolean(A=GetValueAsBool!(Input:"Scalable Float Structure (by ref)"=@n2, Level:float=0) -> (ReturnValue:bool), B=NOT Boolean(A=@n1)))
   > n4 = BindHUDPassiveGA!(execute, self) -> (then, bSuccess:bool) {is=title K2Node_CallFunction}
   > n5 = Branch(Condition=@n4)
   n6 = Get BypassHUDPassiveTimer() -> (BypassHUDPassiveTimer:Timer Handle Structure)
   Branch(execute=[@n5.then, @n3.else], Condition=Is Valid Timer Handle!(Handle:Timer Handle Structure=@n6) -> (ReturnValue:bool))
   > Clear and Invalidate Timer by Handle(Handle=@n6)
   Event Construct()
   > n7 = BindHUDPassiveGA()
   > n8 = Branch(Condition=@n7)
   n9 = Cast To B_WildEstate_CoveSpice!(
    execute=[@n8.then, @n14.then]
    Object=GetEquippedWeapon!(self:Fort Weapon Reticle Extension Widget Base Object Reference) -> (ReturnValue:Fort Weapon Object Reference)
   ) -> (then, CastFailed, AsB Wild Estate Cove Spice:B Wild Estate Cove Spice Object Reference) {is=title K2Node_DynamicCast}
   > n10 = Set WeaponRef(WeaponRef:B Wild Estate Cove Spice Object Reference=@n9) -> (Output_Get:B Wild Estate Cove Spice Object Reference)
   n11 = Create Event()
   > Bind Event to On Overheat Value Changed(self:Fort Weapon Ranged Object Reference=@n10, Delegate=@n11) {is=title K2Node_AddDelegate}
   n12 = Create Event()
   > Bind Event to On Start Firing(self:B Ranged Generic Object Reference=Get WeaponRef() -> (WeaponRef:B Wild Estate Cove Spice Object Reference), Delegate=@n12) {is=title K2Node_AddDelegate}
   > K2Node_Composite_1 = Init Overheat Meter!(execute) -> (then:Exec) {is=title K2Node_Composite}
   Branch(execute=@n8.else, Condition=GetValueAsBool(Input=Get ShouldBypassHUDPassiveFailure(), Level=0))
   > n13 = Set Timer by Event(
    Delegate=Create Event()
    Time=GetValueAtLevel!(
     Input:"Scalable Float Structure (by ref)"=Get BypassHUDPassiveLoopTime() -> (BypassHUDPassiveLoopTime:Scalable Float Structure)
     Level:float=0
    ) -> (ReturnValue:float)
    bLooping=true
   )
   > n14 = Set BypassHUDPassiveTimer(BypassHUDPassiveTimer:Timer Handle Structure=@n13) -> (Output_Get:Timer Handle Structure)
   n15 = Event OnVisibilitySetEvent!() -> (OutputDelegate, then, InVisibility:ESlateVisibility Enum=Visible) {is=title K2Node_Event}
   > Parent: OnVisibilitySetEvent!(execute, InVisibility:ESlateVisibility Enum=@n15.InVisibility) -> (then) {is=title K2Node_CallParentFunction}
   > n16 = Switch on ESlateVisibility!(execute, Selection:ESlateVisibility Enum=@n15.InVisibility) -> (Visible:Exec, Collapsed:Exec, Hidden:Exec, HitTestInvisible:Exec, SelfHitTestInvisible:Exec) {is=title K2Node_SwitchEnum}
   K2Node_Composite_0 = Refresh Overheat Meter!(execute=[@n16.HitTestInvisible, @n16.SelfHitTestInvisible, @n16.Visible]) -> (then:Exec) {is=title K2Node_Composite}
   Event Destruct()
   n17 = Get WeaponRef()
   > n18 = Branch(Condition=IsValid(Object=@n17))
   > Unbind Event from On Overheat Value Changed(self:Fort Weapon Ranged Object Reference=@n17, Delegate=@n11) {is=title K2Node_RemoveDelegate}
   > n19 = Unbind Event from On Start Firing(self:B Ranged Generic Object Reference=@n17, Delegate=@n12) {is=title K2Node_RemoveDelegate}
   n20 = Get HUDPassive()
   n21 = Branch(execute=[@n19.then, @n18.else], Condition=IsValid(Object=@n20))
   > n22 = Unbind Event from On Update Reticle(self:GA Cove Spice HUD Passive Base Object Reference=@n20, Delegate=Create Event()) {is=title K2Node_RemoveDelegate}
   n23 = Get BypassHUDPassiveTimer()
   Branch(execute=[@n22.then, @n21.else], Condition=Is Valid Timer Handle(Handle=@n23))
   > Clear and Invalidate Timer by Handle(Handle=@n23)
  ]}
 Refresh Overheat Meter {
  path=EventGraph.K2Node_Composite_0.Refresh Overheat Meter
  nodes [
   Inputs()
   > n1 = GetDynamicMaterial(self=Get Image_Cooldown())
   n2 = Get WeaponRef()
   n3 = SafeDivide(
    A=GetCurrentOverheatValue!(self:Fort Weapon Ranged Object Reference=@n2) -> (ReturnValue:float)
    B=GetOverheatingMaxValue!(self:Fort Weapon Ranged Object Reference=@n2) -> (ReturnValue:float)
   )
   > SetScalarParameterValue(self=@n1, ParameterName=AnimateProgress, Value=@n3)
   > PrintString!(execute, InString="To String (Float)"(InDouble=@n3), bPrintToScreen, bPrintToLog, TextColor, Duration, Key) -> (then)
   > Outputs!(then:Exec)
  ]
 }
 Init Overheat Meter {
  path=EventGraph.K2Node_Composite_1.Init Overheat Meter
  nodes [
   Inputs()
   > n1 = GetDynamicMaterial(self=Get Image_Cooldown())
   n2 = Get WeaponRef()
   n3 = SafeDivide(A=GetCurrentOverheatValue(self=@n2), B=GetOverheatingMaxValue(self=@n2))
   > SetScalarParameterValue(self=@n1, ParameterName=AnimateProgress, Value=@n3)
   > SetVisibility(self=Get Image_Cooldown(), InVisibility=Select(Option 0=HitTestInvisible, Option 1=Collapsed, Index="Equal (Float)"(A=@n3)))
   > Outputs()
  ]
 }
 HandleInRange {nodes [
   n1 = HandleInRange!() -> (then, Target Location In Range:bool=~) {is=title K2Node_FunctionEntry}
   > n2 = Delta Gate Wildcard!(execute:Exec, Variable:"Boolean (by ref)"=Get InRange() -> (InRange:bool), New Value:bool=@n1) -> (Success:Exec, Updated Variable:"Boolean (by ref)", Previous Variable:bool=~) {is=title K2Node_MacroInstance}
   > n3 = Branch(Condition=@n2)
   > PlayAnimationForward(InAnimation=Get Anim_InRange() -> (Anim_InRange:Widget Animation Object Reference))
   PlayAnimationReverse(execute=@n3.else, InAnimation=Get Anim_InRange())
  ]}
 HandleOverheat {nodes [
   n1 = HandleOverheat!() -> (then, CurrentValue:float=~, MaxValue:float=~, State:EFortWeaponOverheatState Enum=None, bIsInCooldown:bool=~) {is=title K2Node_FunctionEntry}
   > SetVisibility(self=Get Image_Cooldown(), InVisibility=Visible)
   > n2 = GetDynamicMaterial(self=Get Image_Cooldown())
   n3 = SafeDivide(A=@n1, B=@n1.MaxValue)
   > SetScalarParameterValue(self=@n2, ParameterName=AnimateProgress, Value=@n3)
   > Branch(Condition="Equal (Float)"(A=@n3))
   > PlayAnimationForward(InAnimation=Get Anim_HideCooldown())
  ]}
]
variables [
 BypassHUDPassiveLoopTime {
  type=ScalableFloat
  default {
   registryType=None
   value=1
   curve {curveTable=/WildEstateWeapons/DataTables/WildEstateWeaponsGameData, rowName=Default.CoveSpice.HUDPassiveBypassLoopTime}
  }
 }
 BypassHUDPassiveTimer {type=TimerHandle}
 HUDPassive {type=GA_CoveSpice_HUD_Passive_Base_C, default=null}
 InRange {type=boolean}
 PlayerPawn {type=PlayerController, default=null}
 ShouldBypassHUDPassiveFailure {
  type=ScalableFloat
  default {
   registryType=None
   value=1
   curve {curveTable=&WildEstateWeaponsGameData, rowName=Default.CoveSpice.EnableHUDPassiveBypass}
  }
 }
 WeaponRef {type=B_WildEstate_CoveSpice_C, default=null}
]
widgets {
 info {
  inheritedWidgetCount=1
  parentClass=/Script/FortniteUI.FortWeaponReticleExtensionWidgetBase
  rootWidgetClass=&Overlay
  widgetCount=4
 }
 widgets [
  Overlay_0 {widgetClassPath=&Overlay, widget=Overlay_0}
  Image_Cooldown {
   widgetClassPath=&Image
   bIsVariable=true
   parent=Overlay_0
   widget=Image_Cooldown
   properties {
    visibility=Collapsed
    brush {
     bIsDynamicallyLoaded=false
     drawAs=Image
     imageType=FullColor
     mirroring=NoMirror
     resourceName=None
     resourceObject=/WildEstateWeapons/UI/CoveSpice/Materials/MI_UI_CoveSpice_Overheat
     tiling=NoTile
     imageSize=<80 80>
     margin=<ltrb 0 0 0 0>
     outlineSettings {
      bUseBrushTransparency=false
      roundingType=HalfHeightRadius
      width=0
      color {colorUseRule=UseColor_Specified, specifiedColor=<rgba 0 0 0 0>}
      cornerRadii=<0 0 0 0>
     }
     tintColor {colorUseRule=UseColor_Specified, specifiedColor=<rgba 1 1 1 1>}
     uVRegion {bIsValid=false, max=<0 0>, min=<0 0>}
    }
    renderTransform {angle=0, scale=<1 1>, shear=<0 0>, translation=<64 48>}
   }
   slot {class=OverlaySlot}
  }
  Image_Reticle {
   widgetClassPath=&Image
   bIsVariable=true
   parent=Overlay_0
   widget=Image_Reticle
   properties {brush {
     bIsDynamicallyLoaded=false
     drawAs=Image
     imageType=FullColor
     mirroring=NoMirror
     resourceName=None
     resourceObject=/WildEstateWeapons/UI/CoveSpice/Materials/M_UI_CoveSpice_Reticle
     tiling=NoTile
     imageSize=<32 32>
     margin=<ltrb 0 0 0 0>
     outlineSettings {
      bUseBrushTransparency=false
      roundingType=HalfHeightRadius
      width=0
      color {colorUseRule=UseColor_Specified, specifiedColor=<rgba 0 0 0 0>}
      cornerRadii=<0 0 0 0>
     }
     tintColor {colorUseRule=UseColor_Specified, specifiedColor=<rgba 1 1 1 1>}
     uVRegion {bIsValid=false, max=<0 0>, min=<0 0>}
    }}
   slot {class=OverlaySlot, properties {horizontalAlignment=HAlign_Center, verticalAlignment=VAlign_Center}}
  }
  TutorialHighlight {widgetClassPath=&TutorialHighlight}
 ]
}
